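# Masked lit master material for PistolCanary castle debris mesh particles, which glow and dissolve at a portal.
# A sphere mask from PortalCenter and the collection's PortalRadius drives a dithered height fade hiding debris below portal height.
# Emissive glow combines a color with panning world-aligned chromatic caustics, fading with height and favoring downward-facing normals.
# Base textures take optional tiled cracks that darken diffuse and blend normals, plus switchable mask channel order.
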Material /PistolCanary/Assets/FX/Castle/M_PistolCanary_CastleDebris_Master
graph {
 properties {bAutomaticallySetUsageInEditor=false, bUsedWithNiagaraMeshParticles=true, blendMode=BLEND_Masked, usedShadingModels=Default Lit}
 nodes [
  n1 = TextureSampleParameter2D() {parameterName=SpecularMasks, samplerType=SAMPLERTYPE_LinearColor, texture=&/T_CastleMerged_Base_M}
  MaterialFunctionCall(Float3=StaticSwitchParameter(
    True=MaterialFunctionCall(R=@n1.R, G=@n1.R, B=@n1.R) {materialFunction=&AppendMany}.RGB
    False=MaterialFunctionCall(R=@n1.G, G=@n1.G, B=@n1.G) {materialFunction=&AppendMany}.RGB
   ) {defaultValue=true, parameterName=AlternateMaskOrder}) {materialFunction=&BreakOutFloat3Components}
  n2 = ScalarParameter() {defaultValue=700, parameterName=ChromaTile}
  n3 = Multiply(A=@n2, B=ScalarParameter() {defaultValue=-0.5, parameterName=ChromaPanSpeed})
  n4 = Multiply(A=Time(), B=Constant3Vector() {constant=<rgba 0 0 1 0>})
  n5 = MaterialFunctionCall(
   TextureObject=TextureObject() {texture=/Game/Effects/Fort_Effects/Textures/Tile/T_ShieldCaustics_Tile_Chromatic}
   TextureSize=Multiply(A=Constant3Vector() {constant=<rgba 1 1 2 0>}, B=@n2)
   WorldPosition=Add(A=WorldPosition().XYZ, B=Multiply(A=@n3, B=@n4))
  ) {materialFunction=&WorldAlignedTexture}
  n6 = VectorParameter() {parameterName=ChromaColor, defaultValue=<rgba 3 1.0312 1.6361 1>}
  n7 = NamedRerouteUsage()
  n8 = Saturate(None=DotProduct(A=Transform(None=@n7), B=Constant3Vector() {constant=<rgba 0 0 -0.5 0>}))
  n9 = NamedRerouteUsage()
  n10 = VectorParameter() {parameterName=PortalCenter, defaultValue=<rgba 14328 70128 3896 0>}
  n11 = WorldPosition()
  n12 = Subtract(A=@n11.Z, B=@n10.B)
  n13 = Divide(A=@n12, B=ScalarParameter() {defaultValue=1000, parameterName=GlowSize})
  n14 = Power(Base=OneMinus(None=Saturate(None=@n13))) {constExponent=5}
  Multiply(
   A=Add(A=VectorParameter() {parameterName=GlowColor, defaultValue=<rgba 3 0.2296 1.0446 1>}.RGB, B=Multiply(A=@n5.XY Texture, B=@n6.RGB))
   B=Multiply(A=LinearInterpolate(Alpha=@n8) {constA=0.5}, B=Multiply(A=@n9, B=@n14))
  )
  n15 = TextureSampleParameter2D() {parameterName=Normal, samplerType=SAMPLERTYPE_Normal, texture=/Engine/EngineMaterials/BaseFlattenNormalMap}
  n16 = Multiply(A=TextureCoordinate(), B=ScalarParameter() {defaultValue=7, parameterName=CrackTile})
  n17 = TextureSampleParameter2D(UVs=@n16) {parameterName=CrackNormal, samplerType=SAMPLERTYPE_Normal, texture=/Game/Athena/Apollo/Environments/Landscape/Textures/Roads/Road_Cracks_N}
  NamedRerouteDeclaration(None=StaticSwitchParameter(
    True=MaterialFunctionCall(BaseNormal=@n15.RGB, AdditionalNormal=@n17.RGB) {materialFunction=&BlendAngleCorrectedNormals}.Result
    False=@n15.RGB
   ) {defaultValue=true, parameterName=UseCracks}) {name=Normal, nodeColor=<rgba 0 0.2031 1 1>}
  NamedRerouteDeclaration(None=SphereMask(
    A=@n11.XY
    B=ComponentMask(None=@n10.RGB) {g=true, r=true}
    Radius=CollectionParameter() {collection=/PistolCanary/Layout/Material/PostProcess/PistolCanary_MPC, parameterName=PortalRadius}
   ) {hardnessPercent=95}) {name=PortalMask, nodeColor=<rgba 0.5469 1 0 1>}
  n18 = Saturate(None=Divide(A=@n12, B=ScalarParameter() {defaultValue=200, parameterName=MaskFadeSize}) {constB=100})
  n19 = NamedRerouteUsage()
  NamedRerouteDeclaration(None=MaterialFunctionCall(Alpha Threshold=LinearInterpolate(B=@n18, Alpha=@n19) {constA=1}) {materialFunction=&DitherTemporalAA}.Result) {name=OpacityMask, nodeColor=<rgba 0 1 0.3906 1>}
  NamedRerouteUsage()
  n20 = TextureSampleParameter2D() {parameterName=Diffuse, texture=&/T_CastleMerged_Base_D}
  n21 = ScalarParameter() {defaultValue=0.4, parameterName=CrackMinBrightness}
  n22 = Power(Base=TextureSample(UVs=@n16) {samplerType=SAMPLERTYPE_Masks, texture=/Game/Athena/Apollo/Environments/Landscape/Textures/Roads/Road_Cracks_M}.G) {constExponent=0.5}
  StaticSwitchParameter(True=Multiply(A=@n20.RGB, B=LinearInterpolate(A=@n21, Alpha=@n22) {constA=0.3}), False=@n20.RGB) {defaultValue=true, parameterName=UseCracks}
 ]
}
